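# Niagara system for the glow effect of the Magnetic emote, a short-lived red sprite attached to a skeletal mesh.
# A single emitter spawns an instantaneous burst with a camera offset; its continuous spawn rate is disabled.
# Particles re-sample the system-level skeletal mesh location every update with a slight Z offset, so the glow follows the animation.
# Curves grow the sprite from zero and fade its alpha out over its lifetime.
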NiagaraSystem /BRCosmetics/Animation/Game/MainPlayer/Emotes/Magnetic/FX/NS_Emote_Magnetic_Glow
dependencies {
 usedDataInterfaces [&NiagaraDataInterfaceCurve, &NiagaraDataInterfaceSkeletalMesh]
 usedDynamicInputs []
 usedModules [
  &self:SystemSpawnScript.SystemScriptSource.SystemScriptGraph.NiagaraNodeAssignment_0.SetVariables_#1
  &SystemState
 ]
}
emitters {Empty {
  emitterUpdateScript [
   &EmitterState(Life Cycle Mode=::NewEnumerator1, Loop Duration=1, UseLoopDelay=false)
   &SpawnBurst_Instantaneous(Use Spawn Probability=false, UseLoopCountLimit=false)
   &SpawnRate(SpawnRate=2, Use Spawn Probability=false) {enabled=false}
  ]
  particleSpawnScript [
   &InitializeParticle(
    Color=<rgba 2 0 0.0335 0.8>
    Lifetime=0.8
    Position Mode=::NewEnumerator0
    Sprite Rotation Mode=::NewEnumerator2
    Sprite Size Mode=::NewEnumerator3
    Uniform Sprite Size=20
    Write Material Random=true
   )
   &CameraOffset(Camera Offset Amount=4)
  ]
  particleUpdateScript [
   &ParticleState()
   &ScaleSpriteSize(Uniform Curve Sprite Scale=DataInterface() {propertyValues {Curve {keys [{interpMode=RCIM_Cubic, tangentMode=RCTM_User, time=0, value=0}, {interpMode=RCIM_Cubic, tangentMode=RCTM_User, time=0.7, value=1}]}}})
   &SkeletalMeshLocation(Bone Sampling Mode=::NewEnumerator4, MaskSpawnGroup=false, Sampled Position Offset=<0 0 4>, Skeletal Mesh=$System.SkeletalMesh)
   &ScaleColor(
    Scale Alpha=&FloatFromCurve(FloatCurve=DataInterface() {propertyValues {Curve {keys [
         {interpMode=RCIM_Cubic, leaveTangent=-2.0562, tangentMode=RCTM_Break, time=0.35, value=1}
         {arriveTangent=0.0582, interpMode=RCIM_Cubic, tangentMode=RCTM_Break, time=1, value=0}
        ]}}})
    ScaleA=true
    ScaleRGB=true
   )
  ]
  rendererClasses [&NiagaraSpriteRendererProperties]
  renderers [{rendererClass=&NiagaraSpriteRendererProperties}]
 }}
summary {
 systemName=NS_Emote_Magnetic_Glow
 emitters [{emitterName=Empty, rendererClasses [&NiagaraSpriteRendererProperties]}]
 userVariables []
}
